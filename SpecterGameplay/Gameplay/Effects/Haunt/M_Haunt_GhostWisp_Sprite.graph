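# Translucent unlit Niagara particle material that draws a glowing, flickering ghost wisp for the Haunt effect.
# Opacity combines a soft sphere-mask body with a tail of panning dissolve texture, randomly offset per particle and wobbled by a sine wave, then applies depth and camera fades.
# Emissive multiplies a bright core, two-color radial gradient and brightness.
# Named reroutes share center-scaled UVs and a flicker value built from three summed sines.
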Material /SpecterGameplay/Gameplay/Effects/Haunt/M_Haunt_GhostWisp_Sprite
graph {
 properties {
  bAutomaticallySetUsageInEditor=false
  bUsedWithNiagaraMeshParticles=true
  bUsedWithNiagaraSprites=true
  blendMode=BLEND_Translucent
  shadingModel=MSM_Unlit
  usedShadingModels=Unlit
 }
 nodes [
  n1 = NamedRerouteUsage()
  n2 = LinearInterpolate(
   A=ScalarParameter() {defaultValue=0.6, parameterName=BodyCore_Min}
   B=ScalarParameter() {defaultValue=1, parameterName=BodyCore_Max}
   Alpha=SphereMask(A=NamedRerouteUsage(), B=Constant() {r=0.5}, Radius=LinearInterpolate(Alpha=@n1) {constA=0.15, constB=0.25}) {attenuationRadius=0.25, hardnessPercent=0}
  ) {constA=0.6, constB=10}
  n3 = NamedRerouteUsage()
  n4 = LinearInterpolate(
   A=VectorParameter() {parameterName=Color02, defaultValue=<rgba 0.2199 0 0.7552 1>}.RGB
   B=VectorParameter() {parameterName=Color01, defaultValue=<rgba 0 0.7735 2 1>}.RGB
   Alpha=SphereMask(A=NamedRerouteUsage(), B=Constant() {r=0.5}, Radius=LinearInterpolate(Alpha=@n3) {constA=0.7, constB=0.4}) {attenuationRadius=0.5, hardnessPercent=0}
  ) {desc=Gradient Color Layer}
  MaterialFunctionCall(Emissive=Multiply(A=Multiply(A=@n2, B=@n4), B=ScalarParameter() {defaultValue=1, parameterName=Brightness})) {materialFunction=&MF_FX_EmissiveAdjustments}
  n5 = TextureCoordinate()
  n6 = Add(A=Multiply(A=ComponentMask(None=@n5) {g=true}) {constB=3}) {constB=-0.6}
  n7 = OneMinus(None=Saturate(None=Add(A=@n6) {constB=-0.2}))
  n8 = SphereMask(A=NamedRerouteUsage(), B=Constant() {r=0.5}, Hardness=ScalarParameter() {defaultValue=0.3, parameterName=Hardness}) {attenuationRadius=0.35, hardnessPercent=80}
  n9 = Saturate(None=@n6)
  n10 = VectorParameter() {parameterName=Holes Tile and Pan, defaultValue=<rgba 2 0.75 0.05 -0.5>}
  n11 = Panner(Coordinate=Multiply(A=TextureCoordinate(), B=AppendVector(A=@n10.R, B=@n10.R)), Speed=AppendVector(A=@n10.B, B=@n10.B))
  n12 = ParticleRandom()
  n13 = OneMinus(None=TextureSample(UVs=Add(A=@n11, B=@n12)) {texture=/Game/CharClasses/Outlander/Gadgets/AntiMaterialCharge/Materials/T_AntiMaterialCharge_Dissolve_01}.R)
  n14 = MaterialFunctionCall() {materialFunction=&LinearGradient}
  n15 = Multiply(A=Constant() {r=0.4}, B=Power(Base=OneMinus(None=@n14.VGradient)) {constExponent=3})
  n16 = OneMinus(None=MaterialFunctionCall() {materialFunction=&LinearGradient}.VGradient)
  n17 = Multiply(
   A=Add(A=Add(A=@n13, B=@n15) {desc=Adding some gradient for masking}, B=Constant() {desc=Tweaking texture brightness}) {constB=0.25}
   B=Saturate(None=Multiply(A=@n16) {constB=2}) {desc=Tail Bottom Fade}
  )
  n18 = ScalarParameter() {defaultValue=0.065, desc=0.6, parameterName=Texture Subtract}
  n19 = Saturate(None=Divide(A=Subtract(A=@n17, B=@n18), B=ScalarParameter() {defaultValue=0.5, parameterName=Texture Divide}))
  n20 = ComponentMask(None=NamedRerouteUsage()) {g=true}
  n21 = Saturate(None=Add(A=Multiply(A=@n20) {constB=5}) {constB=-2.5})
  n22 = TextureCoordinate()
  n23 = Multiply(A=Constant() {r=-1}, B=Time())
  n24 = ComponentMask(None=TextureCoordinate()) {g=true}
  n25 = Multiply(A=Constant() {r=2}, B=@n24)
  n26 = Add(A=Multiply(A=@n24) {constB=3}) {constB=-1}
  n27 = Multiply(
   A=Sine(None=Add(A=@n23, B=@n25))
   B=Multiply(A=Saturate(None=@n26), B=ScalarParameter() {defaultValue=0.065, parameterName=TailWobble_Intensity})
  )
  n28 = ComponentMask(None=TextureCoordinate()) {g=true}
  n29 = SphereMask(
   A=ComponentMask(None=Add(A=@n22, B=@n27)) {r=true}
   B=Constant() {r=0.5}
   Radius=LinearInterpolate(Alpha=Multiply(A=@n28, B=@n28)) {constA=0.35, constB=0.15}
   Hardness=ScalarParameter() {defaultValue=0.3, parameterName=Hardness}
  ) {attenuationRadius=0.35, hardnessPercent=25}
  n30 = Add(
   A=Multiply(A=Multiply(A=@n8, B=@n7) {desc=Faded Body}) {constB=1.25}
   B=Multiply(A=Multiply(A=@n9, B=@n19), B=Multiply(A=@n21, B=@n29)) {desc=Textured Tail}
  )
  Multiply(
   A=DepthFade(Opacity=Saturate(None=@n30), FadeDistance=ScalarParameter() {defaultValue=5, parameterName=DepthFade_Distance})
   B=MaterialFunctionCall(
    Fade Length=ScalarParameter() {defaultValue=100, parameterName=CamFade_FadeLength}
    Fade Offset=ScalarParameter() {defaultValue=75, parameterName=CamFade_DepthOffset}
   ) {materialFunction=&CameraDepthFade}.Result
  )
  n31 = ScalarParameter() {defaultValue=1, parameterName=FlickerSpeed}
  n32 = Add(
   A=MaterialFunctionCall(Speed=@n31) {materialFunction=&DebugTimeSine}.Result
   B=MaterialFunctionCall(Speed=Multiply(A=@n31) {constB=2.479}) {materialFunction=&DebugTimeSine}.Result
  )
  n33 = MaterialFunctionCall(Speed=Multiply(A=@n31) {constB=1.77}) {materialFunction=&DebugTimeSine}
  NamedRerouteDeclaration(None=Multiply(A=Add(A=@n32, B=@n33.Result)) {constB=0.3333}) {name=FlickerValue, nodeColor=<rgba 1 0.7969 0 1>}
  n34 = TextureCoordinate()
  n35 = AppendVector(A=LinearInterpolate(Alpha=NamedRerouteUsage()) {constA=0.9, constB=1.1}, B=Constant() {r=1})
  NamedRerouteDeclaration(None=Add(
    A=MaterialFunctionCall(UVs=@n34, Texture Scale=@n35) {materialFunction=&ScaleUVsByCenter}.UVs
    B=Constant2Vector() {g=0.14}
   )) {name=UVs_Modified, nodeColor=<rgba 1 0.0523 0 1>}
 ]
}
